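# Sound cue for an ambient spot in the Asteria meteor cave interior, playing electrical glitch sounds.
# A Mixer layers two randomly picked waves: a quieter, lower-pitched electrical ripple doppler and a louder, pitch-modulated glitch one-shot.
# All four quality levels share that mixer.
# The cue uses the indoor ambient sound class, custom distance attenuation with filtering, a small voice limit and a reverb send.
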SoundCue /Game/Sounds/Asteria/POI/Meteor/Spots/Spot_Meteor_Cave_Interior_Glitches_Cue
root=QualityLevel_0
nodes [
 {id=Mixer_0, children [Modulator_1,Modulator_0], properties {inputVolume [0.6,1.25]}}
 {id=Modulator_0, children [Random_1], properties {pitchMax=1.1}}
 {id=Modulator_1, children [Random_2], properties {pitchMax=0.9, pitchMin=0.8}}
 {id=QualityLevel_0, children [Mixer_0,Mixer_0,Mixer_0,Mixer_0]}
 {
  id=Random_1
  children [WavePlayer_3, WavePlayer_4, WavePlayer_11, WavePlayer_12, WavePlayer_13, WavePlayer_14, WavePlayer_15]
  properties {weights [1,1,1,1,1,1,1]}
 }
 {id=Random_2, children [WavePlayer_0,WavePlayer_1,WavePlayer_2,WavePlayer_5,WavePlayer_6,WavePlayer_7], properties {weights [1,1,1,1,1,1]}}
 {id=WavePlayer_0, properties {soundWaveAssetPtr=&/Spot_Meteor_Interior_Electrical_Ripple_Doppler_01}}
 {id=WavePlayer_1, properties {soundWaveAssetPtr=&/Spot_Meteor_Interior_Electrical_Ripple_Doppler_02}}
 {id=WavePlayer_11, properties {soundWaveAssetPtr=&/Meteor_Glitch_Oneshot_05}}
 {id=WavePlayer_12, properties {soundWaveAssetPtr=&/Meteor_Glitch_Oneshot_04}}
 {id=WavePlayer_13, properties {soundWaveAssetPtr=&/Meteor_Glitch_Oneshot_03}}
 {id=WavePlayer_14, properties {soundWaveAssetPtr=&/Meteor_Glitch_Oneshot_02}}
 {id=WavePlayer_15, properties {soundWaveAssetPtr=&/Meteor_Glitch_Oneshot_01}}
 {id=WavePlayer_2, properties {soundWaveAssetPtr=&/Spot_Meteor_Interior_Electrical_Ripple_Doppler_03}}
 {id=WavePlayer_3, properties {soundWaveAssetPtr=&/Meteor_Glitch_Oneshot_07}}
 {id=WavePlayer_4, properties {soundWaveAssetPtr=&/Meteor_Glitch_Oneshot_06}}
 {id=WavePlayer_5, properties {soundWaveAssetPtr=&/Spot_Meteor_Interior_Electrical_Ripple_Doppler_04}}
 {id=WavePlayer_6, properties {soundWaveAssetPtr=&/Spot_Meteor_Interior_Electrical_Ripple_Doppler_05}}
 {id=WavePlayer_7, properties {soundWaveAssetPtr=&/Spot_Meteor_Interior_Electrical_Ripple_Doppler_06}}
]
properties {
 attenuationSettings=&Fort_Ambient_Athena_Att
 bOverrideAttenuation=true
 bOverrideConcurrency=true
 maxDistance=11000
 soundClassObject=&Fort_Ambient_Indoors_Athena
 volumeMultiplier=0.351
 attenuationOverrides {
  bAttenuateWithLPF=true
  bEnableLogFrequencyScaling=true
  falloffDistance=8000
  hPFFrequencyAtMax=1500
  hPFFrequencyAtMin=750
  lPFFrequencyAtMin=7500
  lPFRadiusMax=8000
  stereoSpread=500
  attenuationShapeExtents {x=3000}
 }
 concurrencyOverrides {maxCount=4, voiceStealReleaseTime=2}
 soundSubmixSends [{maxSendDistance=6000, minSendDistance=3000, minSendLevel=0.6, soundSubmix=&Fortnite_Reverb_Default}]
}
