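# Looping ambient sound cue for the interior of the Asteria Meteor POI cave, routed to the indoor ambient sound class at reduced volume.
# A quality-level root selects between two mixers, each blending a glitchy meteor rumble, a reused haunted-house basement bed and a rock-destruction rumble.
# The first two quality levels use the quad rock rumble, while the last two use the stereo version.
# Spatial behavior references the shared Athena ambient attenuation asset.
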SoundCue /Game/Sounds/Asteria/POI/Meteor/Spots/Spot_Meteor_Cave_Interior_Loop_Cue
root=QualityLevel_0
nodes [
 {id=Mixer_0, children [WavePlayer_3,WavePlayer_2,WavePlayer_1], properties {inputVolume [0.75,1.5,1]}}
 {id=Mixer_2, children [WavePlayer_2,WavePlayer_1,WavePlayer_0], properties {inputVolume [1.5,1,0.75]}}
 {id=QualityLevel_0, children [Mixer_0,Mixer_0,Mixer_2,Mixer_2]}
 {id=WavePlayer_0, properties {bLooping=true, soundWaveAssetPtr=/Game/Sounds/Asteria/POI/Meteor/Ambience_Rock_Destruction_Rumble_Loop_Stereo}}
 {
  id=WavePlayer_1
  properties {bLooping=true, soundWaveAssetPtr=/Game/Sounds/Artemis/POI/HauntedHouse/HauntedHouse_AmbientLoop_Basement_Main_02}
 }
 {id=WavePlayer_2, properties {bLooping=true, soundWaveAssetPtr=&/Spot_Meteor_Main_Rumble_Glitch_Loop}}
 {id=WavePlayer_3, properties {bLooping=true, soundWaveAssetPtr=/Game/Sounds/Asteria/POI/Meteor/Ambience_Rock_Destruction_Rumble_Loop_Quad}}
]
properties {
 attenuationSettings=&Fort_Ambient_Athena_Att
 maxDistance=2097152
 soundClassObject=&Fort_Ambient_Indoors_Athena
 volumeMultiplier=0.2
 attenuationOverrides {
  bAttenuate=false
  bAttenuateWithLPF=true
  bEnableLogFrequencyScaling=true
  falloffDistance=3000
  lPFFrequencyAtMax=197.5057
  lPFFrequencyAtMin=20511.4102
  lPFRadiusMax=3789.3618
  lPFRadiusMin=1822.3112
  nonSpatializedRadiusEnd=1500
  nonSpatializedRadiusStart=2000
  spatializationAlgorithm=SPATIALIZATION_HRTF
  stereoSpread=2000
  attenuationShapeExtents {x=1500}
 }
}
