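# Looping ambient spot sound for a meteor crystal shard in the Asteria meteor POI.
# All quality levels share one chain: a short random start delay, then two endlessly looping envelopes of different lengths that slowly modulate volume and pitch on the looping wave.
# Huge-prop attenuation, listener focus, a concurrency cap, and distance-scaled sends to reverb and the exterior ambient bus shape playback.
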SoundCue /Game/Sounds/Asteria/POI/Meteor/Spots/Spot_Meteor_Crystal_Shard_Loop_Cue
root=QualityLevel_1
nodes [
 {id=Delay_0, children [Enveloper_3], properties {delayMax=0.3}}
 {
  id=Enveloper_1
  children [WavePlayer_0]
  properties {
   bLoop=true
   bLoopIndefinitely=true
   loopEnd=30
   pitchCurve {editorCurveData {keys [{arriveTangent=-0.0013, interpMode=RCIM_Cubic, time=0}, <at 15 0.95 RCIM_Cubic>, {interpMode=RCIM_Cubic, leaveTangent=0.0013, time=30}]}}
   volumeCurve {editorCurveData {keys [<at 0 1>]}}
  }
 }
 {
  id=Enveloper_3
  children [Enveloper_1]
  properties {
   bLoop=true
   bLoopIndefinitely=true
   loopEnd=20
   pitchCurve {editorCurveData {keys [{time=0}]}}
   volumeCurve {editorCurveData {keys [
      {arriveTangent=-0.01, interpMode=RCIM_Cubic, time=0, value=1}
      <at 10 0.8 RCIM_Cubic>
      {interpMode=RCIM_Cubic, leaveTangent=0.01, time=20, value=1}
     ]}}
  }
 }
 {id=QualityLevel_1, children [Delay_0,Delay_0,Delay_0,Delay_0]}
 {id=WavePlayer_0, properties {bLooping=true, soundWaveAssetPtr=&/Spot_Meteor_Crystal_Shard_Loop}}
]
properties {
 attenuationSettings=/Game/Sounds/Attenuation/Fort_Amb_Emitter_Prop_Huge_Att
 bOverrideConcurrency=true
 maxDistance=10000
 pitchMultiplier=2
 soundClassObject=/Game/Sounds/FortSoundClassesAndModes/Fort_Storms_PSM
 volumeMultiplier=3
 attenuationOverrides {
  bEnableFocusInterpolation=true
  bEnableListenerFocus=true
  falloffDistance=6000
  focusAttackInterpSpeed=0.01
  focusReleaseInterpSpeed=0.01
  nonFocusVolumeAttenuation=0.75
  stereoSpread=400
  attenuationShapeExtents {x=3000}
 }
 busSends [{
   audioBus=null
   maxSendDistance=6000
   minSendDistance=3000
   minSendLevel=0.6
   sendLevel=1
   soundSourceBus=&AmbientAudio_Exterior_SourceBus
  }]
 concurrencyOverrides {maxCount=2, voiceStealReleaseTime=2}
 soundSubmixSends [{maxSendDistance=7000, minSendDistance=3000, minSendLevel=0.3, soundSubmix=&Fortnite_Reverb_Default}]
}
