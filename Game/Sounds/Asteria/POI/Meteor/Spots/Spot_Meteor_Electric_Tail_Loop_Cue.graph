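# Looping sound cue for the electric tail of a meteor spot sound in the Asteria meteor POI.
# One looping wave feeds two capsule attenuations, close and regular, and a mixer layers them with the close one at double volume.
# All quality levels share that mixer.
# The cue uses the Fort_Storms_PSM sound class and sends to the default Fortnite reverb submix.
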SoundCue /Game/Sounds/Asteria/POI/Meteor/Spots/Spot_Meteor_Electric_Tail_Loop_Cue
root=QualityLevel_0
nodes [
 {
  id=Attenuation_0
  children [WavePlayer_0]
  properties {attenuationSettings=&/Meteor_Electric_Tail_Capsule_Close_Att}
 }
 {
  id=Attenuation_1
  children [WavePlayer_0]
  properties {attenuationSettings=&/Meteor_Electric_Tail_Capsule_Att}
 }
 {id=Mixer_1, children [Attenuation_0, Attenuation_1], properties {inputVolume [2,1]}}
 {id=QualityLevel_0, children [Mixer_1,Mixer_1,Mixer_1,Mixer_1]}
 {id=WavePlayer_0, properties {bLooping=true, soundWaveAssetPtr=&/Spot_Meteor_Electric_Tail_Loop}}
]
properties {
 maxDistance=120000
 soundClassObject=/Game/Sounds/FortSoundClassesAndModes/Fort_Storms_PSM
 volumeMultiplier=1
 attenuationOverrides {
  bAttenuateWithLPF=true
  bEnableLogFrequencyScaling=true
  distanceAlgorithm=Logarithmic
  falloffDistance=10000
  lPFFrequencyAtMax=197.5057
  lPFFrequencyAtMin=20511.4102
  lPFRadiusMax=3789.3618
  lPFRadiusMin=1822.3112
  stereoSpread=2000
  attenuationShapeExtents {x=0}
 }
 soundSubmixSends [{minSendLevel=0, soundSubmix=&Fortnite_Reverb_Default}]
}
