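# Looping ambient cue for slow lightning discharge at the Asteria Meteor POI, using spline-named lightning attenuation.
# A constant discharge loop runs through nested 60-second looping envelopers that slowly drift pitch and volume.
# Mixed quieter beneath it, two endless loops fire random thunder cracks and soft distant rumbles after short random delays with pitch variation.
# The cue sends to the exterior ambient bus and reverb submix, with its own concurrency override.
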SoundCue /Game/Sounds/Asteria/POI/Meteor/Spots/Spot_Meteor_Lightening_Slow_Discharge_Cue
root=Mixer_1
nodes [
 {id=Delay_1, children [Random_1], properties {delayMax=3, delayMin=2}}
 {id=Delay_2, children [Random_2], properties {delayMax=3, delayMin=2}}
 {
  id=Enveloper_0
  children [Modulator_0]
  properties {
   bLoop=true
   bLoopIndefinitely=true
   loopEnd=60
   pitchCurve {editorCurveData {keys [
      {interpMode=RCIM_Cubic, time=0}
      <at 15 0.85 RCIM_Cubic>
      {arriveTangent=0.0091, interpMode=RCIM_Cubic, leaveTangent=0.0091, tangentMode=RCTM_User, time=30}
      {interpMode=RCIM_Cubic, time=45}
      {time=60}
     ]}}
   volumeCurve {editorCurveData {keys [<at 0 1>]}}
  }
 }
 {
  id=Enveloper_3
  children [Enveloper_0]
  properties {
   bLoop=true
   bLoopIndefinitely=true
   loopEnd=60
   pitchCurve {editorCurveData {keys [{time=0}]}}
   volumeCurve {editorCurveData {keys [
      {arriveTangent=-0.01, interpMode=RCIM_Cubic, time=0, value=1}
      {arriveTangent=0.0001, interpMode=RCIM_Cubic, leaveTangent=0.0001, tangentMode=RCTM_User, time=15, value=0.8}
      {arriveTangent=0.005, interpMode=RCIM_Cubic, leaveTangent=0.005, time=30, value=1}
      <at 45 0.95 RCIM_Cubic>
      <at 60 1>
     ]}}
  }
 }
 {id=Looping_1, children [Modulator_1]}
 {id=Looping_2, children [Modulator_2]}
 {id=Mixer_1, children [Enveloper_3,Looping_1,Looping_2], properties {inputVolume [0.8,0.5,0.5]}}
 {id=Modulator_0, children [WavePlayer_10], properties {pitchMax=1, pitchMin=1}}
 {id=Modulator_1, children [Delay_1], properties {pitchMax=1.1, pitchMin=0.93, volumeMax=1}}
 {id=Modulator_2, children [Delay_2], properties {pitchMax=1.1, pitchMin=0.93, volumeMax=1}}
 {id=Random_1, children [WavePlayer_11, WavePlayer_12, WavePlayer_13, WavePlayer_14], properties {weights [1,1,1,1]}}
 {
  id=Random_2
  children [
   WavePlayer_15
   WavePlayer_16
   WavePlayer_17
   WavePlayer_18
   WavePlayer_19
   WavePlayer_20
   WavePlayer_21
   WavePlayer_22
   WavePlayer_23
   WavePlayer_24
  ]
  properties {weights [1,1,1,1,1,1,1,1,1,1]}
 }
 {
  id=WavePlayer_10
  properties {bLooping=true, soundWaveAssetPtr=&/Spot_Meteor_Lightening_Slow_Constant_Discharge_Loop}
 }
 {id=WavePlayer_11, properties {soundWaveAssetPtr=&/Spot_Meteor_Thunder_Crack_Swt_04}}
 {id=WavePlayer_12, properties {soundWaveAssetPtr=&/Spot_Meteor_Thunder_Crack_Swt_01}}
 {id=WavePlayer_13, properties {soundWaveAssetPtr=&/Spot_Meteor_Thunder_Crack_Swt_02}}
 {id=WavePlayer_14, properties {soundWaveAssetPtr=&/Spot_Meteor_Thunder_Crack_Swt_03}}
 {id=WavePlayer_15, properties {soundWaveAssetPtr=&/Spot_Meteor_Thunder_Soft_Rumbling_Distant_01}}
 {id=WavePlayer_16, properties {soundWaveAssetPtr=&/Spot_Meteor_Thunder_Soft_Rumbling_Distant_02}}
 {id=WavePlayer_17, properties {soundWaveAssetPtr=&/Spot_Meteor_Thunder_Soft_Rumbling_Distant_03}}
 {id=WavePlayer_18, properties {soundWaveAssetPtr=&/Spot_Meteor_Thunder_Soft_Rumbling_Distant_04}}
 {id=WavePlayer_19, properties {soundWaveAssetPtr=&/Spot_Meteor_Thunder_Soft_Rumbling_Distant_05}}
 {id=WavePlayer_20, properties {soundWaveAssetPtr=&/Spot_Meteor_Thunder_Soft_Rumbling_Distant_06}}
 {id=WavePlayer_21, properties {soundWaveAssetPtr=&/Spot_Meteor_Thunder_Soft_Rumbling_Distant_07}}
 {id=WavePlayer_22, properties {soundWaveAssetPtr=&/Spot_Meteor_Thunder_Soft_Rumbling_Distant_08}}
 {id=WavePlayer_23, properties {soundWaveAssetPtr=&/Spot_Meteor_Thunder_Soft_Rumbling_Distant_09}}
 {id=WavePlayer_24, properties {soundWaveAssetPtr=&/Spot_Meteor_Thunder_Soft_Rumbling_Distant_10}}
]
properties {
 attenuationSettings=&/Fort_Meteor_Lightning_Spline_Att
 bOverrideConcurrency=true
 maxDistance=21000
 soundClassObject=&Fort_Gadgets
 volumeMultiplier=0.7
 attenuationOverrides {
  bAttenuateWithLPF=true
  bEnableLogFrequencyScaling=true
  distanceAlgorithm=Logarithmic
  falloffDistance=10000
  lPFFrequencyAtMax=197.5057
  lPFFrequencyAtMin=20511.4102
  lPFRadiusMax=3789.3618
  lPFRadiusMin=1822.3112
  stereoSpread=2000
  attenuationShapeExtents {x=0}
 }
 busSends [{audioBus=null, minSendLevel=0.6, sendLevel=1, soundSourceBus=&AmbientAudio_Exterior_SourceBus}]
 concurrencyOverrides {maxCount=10, voiceStealReleaseTime=0.5}
 soundSubmixSends [{minSendLevel=0.6, soundSubmix=&Fortnite_Reverb_Default}]
}
