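# Looping ambient Sound Cue for the rumble at the Meteor POI on Asteria, in the Athena ambient sound class.
# A random start delay of up to half a second feeds a quality switch between two mixers.
# Both mixers layer the Meteor rumble loop over a reused SkyIsland rumble, with the Meteor layer slightly louder at the first quality level.
# Overridden attenuation adds HRTF spatialization, a custom falloff curve, distance low-pass filtering and a non-spatialized near radius.
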SoundCue /Game/Sounds/Asteria/POI/Meteor/Spots/Spot_Meteor_Main_Rumble_Sphere_Rock_Loop_Cue
root=Delay_0
nodes [
 {id=Delay_0, children [QualityLevel_0], properties {delayMax=0.5}}
 {id=Mixer_1, children [WavePlayer_5,WavePlayer_2], properties {inputVolume [2,1]}}
 {id=Mixer_3, children [WavePlayer_12, WavePlayer_3], properties {inputVolume [1.75,1]}}
 {id=QualityLevel_0, children [Mixer_1,Mixer_3,Mixer_3,Mixer_3]}
 {id=WavePlayer_12, properties {bLooping=true, soundWaveAssetPtr=&/Spot_Meteor_Main_Rumble_Loop}}
 {id=WavePlayer_2, properties {bLooping=true, soundWaveAssetPtr=/Game/Sounds/Asteria/POI/SkyIsland/Spots/Spot_SkyIsland_Rumble_All}}
 {id=WavePlayer_3, properties {bLooping=true, soundWaveAssetPtr=&Spot_SkyIsland_Rumble_All}}
 {id=WavePlayer_5, properties {bLooping=true, soundWaveAssetPtr=&Spot_Meteor_Main_Rumble_Loop}}
]
properties {
 bOverrideAttenuation=true
 maxDistance=6000
 soundClassObject=&Fort_Ambient_Athena
 volumeMultiplier=0.626
 attenuationOverrides {
  bAttenuateWithLPF=true
  bEnableLogFrequencyScaling=true
  distanceAlgorithm=Custom
  falloffDistance=3000
  lPFFrequencyAtMax=197.5057
  lPFFrequencyAtMin=20511.4102
  lPFRadiusMax=3789.3618
  lPFRadiusMin=1822.3112
  nonSpatializedRadiusEnd=1500
  nonSpatializedRadiusStart=3000
  spatializationAlgorithm=SPATIALIZATION_HRTF
  stereoSpread=1000
  attenuationShapeExtents {x=0}
  customAttenuationCurve {editorCurveData {keys [<at 0 1>,<at 0.75 1>,<at 2 0>]}}
 }
}
