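# Looping ambient spot cue for rock destruction at the Asteria Meteor point of interest.
# A Mixer layers two looping waves at equal volume: a meteor rock destruction loop and a rumble borrowed from SkyIsland.
# The mix passes through a close-distance meteor rumble Attenuation node, which feeds all four QualityLevel branches identically.
# The cue uses the storms sound class and sends to exterior ambient bus and default reverb submix.
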SoundCue /Game/Sounds/Asteria/POI/Meteor/Spots/Spot_Meteor_Rock_Destruction_Loop_Cue
root=QualityLevel_0
nodes [
 {
  id=Attenuation_0
  children [Mixer_0]
  properties {attenuationSettings=&/Meteor_Main_Rock_Rumble_Close_Dist_Att}
 }
 {id=Mixer_0, children [WavePlayer_0,WavePlayer_2], properties {inputVolume [1,1]}}
 {id=QualityLevel_0, children [Attenuation_0, Attenuation_0, Attenuation_0, Attenuation_0]}
 {id=WavePlayer_0, properties {bLooping=true, soundWaveAssetPtr=&/Spot_Meteor_Rock_Desctruction_Loop_03}}
 {id=WavePlayer_2, properties {bLooping=true, soundWaveAssetPtr=/Game/Sounds/Asteria/POI/SkyIsland/Spots/Spot_SkyIsland_Rumble_All}}
]
properties {
 maxDistance=5500
 soundClassObject=/Game/Sounds/FortSoundClassesAndModes/Fort_Storms_PSM
 volumeMultiplier=1
 attenuationOverrides {
  attenuationShape=Capsule
  bAttenuateWithLPF=true
  bEnableLogFrequencyScaling=true
  falloffDistance=10000
  lPFFrequencyAtMax=197.5057
  lPFFrequencyAtMin=20511.4102
  lPFRadiusMax=3789.3618
  lPFRadiusMin=1822.3112
  nonSpatializedRadiusEnd=2500
  nonSpatializedRadiusStart=5000
  stereoSpread=2000
  attenuationShapeExtents=<512.3855 1207.2642>
 }
 busSends [{audioBus=null, minSendLevel=0.4, sendLevel=1, soundSourceBus=&AmbientAudio_Exterior_SourceBus}]
 soundSubmixSends [{maxSendDistance=6000, minSendLevel=0.3, soundSubmix=&Fortnite_Reverb_Default}]
}
