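# Sound cue for a looping fire and destruction rumble, named as a small meteor tail spot sound in the Asteria meteor POI.
# A short random delay and a pitch-varying modulator lead into a quality switch that plays one looping wave on the first two levels, leaving the other two empty.
# It uses the storms sound class, sends to the exterior ambient bus and overrides concurrency to cap simultaneous voices at two.
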SoundCue /Game/Sounds/Asteria/POI/Meteor/Spots/Spot_Meteor_Small_Tail_Fire_Descruction_Rumble_Cue
root=Delay_0
nodes [
 {id=Delay_0, children [Modulator_0], properties {delayMax=0.3}}
 {id=Modulator_0, children [QualityLevel_0], properties {pitchMin=0.9}}
 {id=QualityLevel_0, children [WavePlayer_1,WavePlayer_1,null,null]}
 {
  id=WavePlayer_1
  properties {bLooping=true, soundWaveAssetPtr=&/Spot_Meteor_Small_Tail_Fire_Descruction_Rumble_02}
 }
]
properties {
 bOverrideConcurrency=true
 maxDistance=2097152
 soundClassObject=/Game/Sounds/FortSoundClassesAndModes/Fort_Storms_PSM
 volumeMultiplier=1.255
 busSends [{
   audioBus=null
   maxSendLevel=0.9
   minSendLevel=0.5
   sendLevel=1
   soundSourceBus=&AmbientAudio_Exterior_SourceBus
  }]
 concurrencyOverrides {maxCount=2, voiceStealReleaseTime=2}
}
