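# Looping meteor-site thunder spot ambience for the Asteria POI, routed to the storms sound class.
# Three distance layers with separate attenuations blend close lightning-crack loops, distant thunder loops and randomized distant rumbles.
# One-shots repeat via looping random delays with pitch and volume modulation.
# A quality switch feeds the first two levels a richer mix with ambient beds and extra cracks, the last two a leaner mix with fewer rumbles.
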SoundCue /Game/Sounds/Asteria/POI/Meteor/Spots/Spot_Meteor_Thunder_Loop_Cue
root=QualityLevel_0
nodes [
 {
  id=Attenuation_0
  children [Mixer_5]
  properties {attenuationSettings=&/Meteor_Thunder_Capsule_Close_Att}
 }
 {
  id=Attenuation_1
  children [WavePlayer_26]
  properties {attenuationSettings=&Meteor_Thunder_Capsule_Close_Att}
 }
 {id=Attenuation_2, children [Mixer_4], properties {attenuationSettings=&/Meteor_Thunder_Capsule_Att}}
 {
  id=Attenuation_3
  children [Looping_0]
  properties {attenuationSettings=&/Meteor_Thunder_Med_Distant_Focus_Att}
 }
 {
  id=Attenuation_4
  children [WavePlayer_13]
  properties {attenuationSettings=&Meteor_Thunder_Capsule_Att}
 }
 {
  id=Attenuation_5
  children [Looping_1]
  properties {attenuationSettings=&Meteor_Thunder_Med_Distant_Focus_Att}
 }
 {id=Delay_0, children [Random_0], properties {delayMax=5, delayMin=2}}
 {id=Delay_1, children [Random_1], properties {delayMax=5}}
 {id=Delay_2, children [Random_2], properties {delayMax=5, delayMin=1}}
 {id=Delay_3, children [Random_3], properties {delayMax=4, delayMin=1}}
 {id=Looping_0, children [Modulator_0]}
 {id=Looping_1, children [Modulator_1]}
 {id=Looping_2, children [Modulator_2]}
 {id=Looping_3, children [Modulator_4]}
 {id=Mixer_0, children [WavePlayer_29, WavePlayer_12], properties {inputVolume [3,1.25]}}
 {id=Mixer_1, children [Attenuation_0, Attenuation_2, Attenuation_3], properties {inputVolume [1,0.9,0.9]}}
 {id=Mixer_2, children [WavePlayer_5,Looping_2], properties {inputVolume [1.5,1.5]}}
 {id=Mixer_3, children [Attenuation_1, Attenuation_4, Attenuation_5], properties {inputVolume [1.75,1,1]}}
 {id=Mixer_4, children [Modulator_3,Mixer_0], properties {inputVolume [0.3,1]}}
 {id=Mixer_5, children [Mixer_2, WavePlayer_29], properties {inputVolume [1,1]}}
 {id=Modulator_0, children [Delay_0], properties {pitchMax=1.1, pitchMin=0.93, volumeMax=1}}
 {id=Modulator_1, children [Delay_1], properties {pitchMax=1.1, pitchMin=0.93, volumeMax=1}}
 {id=Modulator_2, children [Delay_2], properties {pitchMax=1.1, pitchMin=0.93, volumeMax=1}}
 {id=Modulator_3, children [Looping_3], properties {pitchMax=0.95, pitchMin=0.9}}
 {id=Modulator_4, children [Delay_3], properties {pitchMax=1, pitchMin=0.8, volumeMax=1}}
 {id=QualityLevel_0, children [Mixer_1,Mixer_1,Mixer_3,Mixer_3]}
 {
  id=Random_0
  children [WavePlayer_0, WavePlayer_1, WavePlayer_2, WavePlayer_4, WavePlayer_6, WavePlayer_7, WavePlayer_8, WavePlayer_9, WavePlayer_10, WavePlayer_11]
  properties {weights [1,1,1,1,1,1,1,1,1,1]}
 }
 {id=Random_1, children [WavePlayer_14, WavePlayer_15, WavePlayer_16], properties {weights [1,1,1]}}
 {
  id=Random_2
  children [WavePlayer_17, WavePlayer_19, WavePlayer_20, WavePlayer_21, WavePlayer_22, WavePlayer_23, WavePlayer_24, WavePlayer_25]
  properties {weights [1,1,1,1,1,1,1,1]}
 }
 {
  id=Random_3
  children [WavePlayer_3, WavePlayer_28, WavePlayer_30, WavePlayer_31, WavePlayer_32, WavePlayer_33, WavePlayer_34, WavePlayer_35]
  properties {weights [1,1,1,1,1,1,1,1]}
 }
 {id=WavePlayer_0, properties {soundWaveAssetPtr=&/Spot_Meteor_Thunder_Soft_Rumbling_Distant_01}}
 {id=WavePlayer_1, properties {soundWaveAssetPtr=&/Spot_Meteor_Thunder_Soft_Rumbling_Distant_02}}
 {id=WavePlayer_10, properties {soundWaveAssetPtr=&/Spot_Meteor_Thunder_Soft_Rumbling_Distant_09}}
 {id=WavePlayer_11, properties {soundWaveAssetPtr=&/Spot_Meteor_Thunder_Soft_Rumbling_Distant_10}}
 {id=WavePlayer_12, properties {bLooping=true, soundWaveAssetPtr=&/Spot_Meteor_Thunder_Distant_Loop}}
 {id=WavePlayer_13, properties {bLooping=true, soundWaveAssetPtr=&Spot_Meteor_Thunder_Distant_Loop}}
 {id=WavePlayer_14, properties {soundWaveAssetPtr=&Spot_Meteor_Thunder_Soft_Rumbling_Distant_01}}
 {id=WavePlayer_15, properties {soundWaveAssetPtr=&Spot_Meteor_Thunder_Soft_Rumbling_Distant_02}}
 {id=WavePlayer_16, properties {soundWaveAssetPtr=&/Spot_Meteor_Thunder_Soft_Rumbling_Distant_03}}
 {id=WavePlayer_17, properties {soundWaveAssetPtr=&/Spot_Lightning_Crack_Close_01}}
 {id=WavePlayer_19, properties {soundWaveAssetPtr=&/Spot_Lightning_Crack_Close_03}}
 {id=WavePlayer_2, properties {soundWaveAssetPtr=&Spot_Meteor_Thunder_Soft_Rumbling_Distant_03}}
 {id=WavePlayer_20, properties {soundWaveAssetPtr=&/Spot_Lightning_Crack_Close_04}}
 {id=WavePlayer_21, properties {soundWaveAssetPtr=&/Spot_Lightning_Crack_Close_05}}
 {id=WavePlayer_22, properties {soundWaveAssetPtr=&/Spot_Lightning_Crack_Close_06}}
 {id=WavePlayer_23, properties {soundWaveAssetPtr=&/Spot_Lightning_Crack_Close_07}}
 {id=WavePlayer_24, properties {soundWaveAssetPtr=&/Spot_Lightning_Crack_Close_08}}
 {id=WavePlayer_25, properties {soundWaveAssetPtr=&/Spot_Lightning_Crack_Close_09}}
 {
  id=WavePlayer_26
  properties {bLooping=true, soundWaveAssetPtr=&/Spot_Meteor_Lightning_Cracks_Close_Loop_02}
 }
 {id=WavePlayer_28, properties {soundWaveAssetPtr=&Spot_Lightning_Crack_Close_03}}
 {id=WavePlayer_29, properties {bLooping=true, soundWaveAssetPtr=&/Meteor_AmbientLoop_01}}
 {id=WavePlayer_3, properties {soundWaveAssetPtr=&Spot_Lightning_Crack_Close_01}}
 {id=WavePlayer_30, properties {soundWaveAssetPtr=&Spot_Lightning_Crack_Close_04}}
 {id=WavePlayer_31, properties {soundWaveAssetPtr=&Spot_Lightning_Crack_Close_05}}
 {id=WavePlayer_32, properties {soundWaveAssetPtr=&Spot_Lightning_Crack_Close_06}}
 {id=WavePlayer_33, properties {soundWaveAssetPtr=&Spot_Lightning_Crack_Close_07}}
 {id=WavePlayer_34, properties {soundWaveAssetPtr=&Spot_Lightning_Crack_Close_08}}
 {id=WavePlayer_35, properties {soundWaveAssetPtr=&Spot_Lightning_Crack_Close_09}}
 {id=WavePlayer_4, properties {soundWaveAssetPtr=&/Spot_Meteor_Thunder_Soft_Rumbling_Distant_04}}
 {
  id=WavePlayer_5
  properties {bLooping=true, soundWaveAssetPtr=&Spot_Meteor_Lightning_Cracks_Close_Loop_02}
 }
 {id=WavePlayer_6, properties {soundWaveAssetPtr=&/Spot_Meteor_Thunder_Soft_Rumbling_Distant_05}}
 {id=WavePlayer_7, properties {soundWaveAssetPtr=&/Spot_Meteor_Thunder_Soft_Rumbling_Distant_06}}
 {id=WavePlayer_8, properties {soundWaveAssetPtr=&/Spot_Meteor_Thunder_Soft_Rumbling_Distant_07}}
 {id=WavePlayer_9, properties {soundWaveAssetPtr=&/Spot_Meteor_Thunder_Soft_Rumbling_Distant_08}}
]
properties {
 maxDistance=120000
 soundClassObject=/Game/Sounds/FortSoundClassesAndModes/Fort_Storms_PSM
 volumeMultiplier=1
 attenuationOverrides {
  bAttenuateWithLPF=true
  bEnableLogFrequencyScaling=true
  distanceAlgorithm=Logarithmic
  falloffDistance=10000
  lPFFrequencyAtMax=197.5057
  lPFFrequencyAtMin=20511.4102
  lPFRadiusMax=3789.3618
  lPFRadiusMin=1822.3112
  stereoSpread=2000
  attenuationShapeExtents {x=0}
 }
 soundSubmixSends [{minSendLevel=0, soundSubmix=&Fortnite_Reverb_Default}]
}
