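# Sound cue for close thunder and lightning stingers at Asteria's Meteor POI, layered as a continuous storm ambience.
# A mixer blends four endlessly repeating random one-shot layers (distant rumbles, thunder cracks, two close lightning sets), each randomly delayed and pitch-modulated, over a looping crack bed.
# The quality node feeds that mixer to its first two slots, leaving two empty.
# Overrides disable distance volume attenuation, add listener focus, and cap concurrency.
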SoundCue /Game/Sounds/Asteria/POI/Meteor/Spots/Spot_Meteor_Thunder_Stingers_Close_Cue
root=QualityLevel_0
nodes [
 {id=Delay_0, children [Random_0], properties {delayMax=5, delayMin=2}}
 {id=Delay_1, children [Random_1], properties {delayMax=6, delayMin=4}}
 {id=Delay_2, children [Random_2], properties {delayMax=3, delayMin=2}}
 {id=Delay_3, children [Random_3], properties {delayMax=3, delayMin=1}}
 {id=Looping_0, children [Modulator_0]}
 {id=Looping_1, children [Modulator_1]}
 {id=Looping_2, children [Modulator_2]}
 {id=Looping_3, children [Modulator_3]}
 {id=Mixer_0, children [Looping_2,Looping_1,Looping_3,Looping_0,Modulator_4], properties {inputVolume [1,2,1,1,0.5]}}
 {id=Modulator_0, children [Delay_1], properties {pitchMax=1.1, pitchMin=0.93, volumeMax=1}}
 {id=Modulator_1, children [Delay_0], properties {pitchMax=1.2, pitchMin=1, volumeMax=1}}
 {id=Modulator_2, children [Delay_2], properties {pitchMax=1.1, pitchMin=0.93, volumeMax=1}}
 {id=Modulator_3, children [Delay_3], properties {pitchMax=1.1, pitchMin=0.93, volumeMax=1}}
 {id=Modulator_4, children [WavePlayer_21], properties {pitchMax=1.2, pitchMin=1, volumeMax=1}}
 {id=QualityLevel_0, children [Mixer_0,Mixer_0,null,null]}
 {
  id=Random_0
  children [
   WavePlayer_0
   WavePlayer_2
   WavePlayer_3
   WavePlayer_4
   WavePlayer_6
   WavePlayer_7
   WavePlayer_8
   WavePlayer_9
   WavePlayer_10
   WavePlayer_11
   WavePlayer_23
  ]
  properties {weights [1,1,1,1,1,1,1,1,1,1,1]}
 }
 {id=Random_1, children [WavePlayer_1, WavePlayer_5, WavePlayer_12, WavePlayer_13], properties {weights [1,1,1,1]}}
 {id=Random_2, children [WavePlayer_14, WavePlayer_15, WavePlayer_16, WavePlayer_17], properties {weights [1,1,1,1]}}
 {id=Random_3, children [WavePlayer_18, WavePlayer_19, WavePlayer_20, WavePlayer_22], properties {weights [1,1,1,1]}}
 {id=WavePlayer_0, properties {soundWaveAssetPtr=&/Spot_Meteor_Thunder_Soft_Rumbling_Distant_01}}
 {id=WavePlayer_1, properties {soundWaveAssetPtr=&/Spot_Meteor_Thunder_Crack_Swt_01}}
 {id=WavePlayer_10, properties {soundWaveAssetPtr=&/Spot_Meteor_Thunder_Soft_Rumbling_Distant_09}}
 {id=WavePlayer_11, properties {soundWaveAssetPtr=&/Spot_Meteor_Thunder_Soft_Rumbling_Distant_10}}
 {id=WavePlayer_12, properties {soundWaveAssetPtr=&/Spot_Meteor_Thunder_Crack_Swt_03}}
 {id=WavePlayer_13, properties {soundWaveAssetPtr=&/Spot_Meteor_Thunder_Crack_Swt_04}}
 {id=WavePlayer_14, properties {soundWaveAssetPtr=&/Spot_Lightning_Crack_Close_09}}
 {id=WavePlayer_15, properties {soundWaveAssetPtr=&/Spot_Lightning_Crack_Close_08}}
 {id=WavePlayer_16, properties {soundWaveAssetPtr=&/Spot_Lightning_Crack_Close_07}}
 {id=WavePlayer_17, properties {soundWaveAssetPtr=&/Spot_Lightning_Crack_Close_06}}
 {id=WavePlayer_18, properties {soundWaveAssetPtr=&/Spot_Lightning_Crack_Close_05}}
 {id=WavePlayer_19, properties {soundWaveAssetPtr=&/Spot_Lightning_Crack_Close_04}}
 {id=WavePlayer_2, properties {soundWaveAssetPtr=&/Spot_Meteor_Thunder_Soft_Rumbling_Distant_02}}
 {id=WavePlayer_20, properties {soundWaveAssetPtr=&/Spot_Lightning_Crack_Close_03}}
 {
  id=WavePlayer_21
  properties {bLooping=true, soundWaveAssetPtr=&/Spot_Meteor_Lightning_Cracks_Close_Loop_02}
 }
 {id=WavePlayer_22, properties {soundWaveAssetPtr=&/Spot_Lightning_Crack_Close_01}}
 {id=WavePlayer_23, properties {soundWaveAssetPtr=&/Spot_Meteor_Thunder_Soft_Rumbling_Distant_11}}
 {id=WavePlayer_3, properties {soundWaveAssetPtr=&/Spot_Meteor_Thunder_Soft_Rumbling_Distant_03}}
 {id=WavePlayer_4, properties {soundWaveAssetPtr=&/Spot_Meteor_Thunder_Soft_Rumbling_Distant_04}}
 {id=WavePlayer_5, properties {soundWaveAssetPtr=&/Spot_Meteor_Thunder_Crack_Swt_02}}
 {id=WavePlayer_6, properties {soundWaveAssetPtr=&/Spot_Meteor_Thunder_Soft_Rumbling_Distant_05}}
 {id=WavePlayer_7, properties {soundWaveAssetPtr=&/Spot_Meteor_Thunder_Soft_Rumbling_Distant_06}}
 {id=WavePlayer_8, properties {soundWaveAssetPtr=&/Spot_Meteor_Thunder_Soft_Rumbling_Distant_07}}
 {id=WavePlayer_9, properties {soundWaveAssetPtr=&/Spot_Meteor_Thunder_Soft_Rumbling_Distant_08}}
]
properties {
 bOverrideAttenuation=true
 bOverrideConcurrency=true
 maxDistance=2097152
 soundClassObject=/Game/Sounds/FortSoundClassesAndModes/Fort_Storms_PSM
 volumeMultiplier=1.5
 attenuationOverrides {
  attenuationShape=Cone
  bAttenuate=false
  bEnableFocusInterpolation=true
  bEnableListenerFocus=true
  falloffDistance=7500
  focusAttackInterpSpeed=0.05
  focusAzimuth=45
  focusDistanceScale=0.5
  focusReleaseInterpSpeed=0.05
  lPFFrequencyAtMax=1500
  lPFRadiusMax=8500
  lPFRadiusMin=3500
  nonFocusVolumeAttenuation=0.75
  stereoSpread=1000
  attenuationShapeExtents=<3000 16 25>
  submixSendSettings [{
    bEnableHPFCutoff=false
    bEnableLPFCutoff=false
    disableManualSendClamp=false
    hPFCutoff=20
    lPFCutoff=20000
    maxSendDistance=6000
    maxSendLevel=1
    minSendDistance=400
    minSendLevel=0
    sendLevel=0.2
    sendLevelControlMethod=Linear
    soundSubmix=null
   }]
 }
 busSends [{
   audioBus=null
   maxSendDistance=6000
   minSendDistance=2000
   minSendLevel=0.6
   sendLevel=1
   soundSourceBus=&AmbientAudio_Exterior_SourceBus
  }]
 concurrencyOverrides {maxCount=2, voiceStealReleaseTime=2}
}
